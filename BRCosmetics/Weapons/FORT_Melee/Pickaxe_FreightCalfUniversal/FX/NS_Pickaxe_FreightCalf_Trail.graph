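# Niagara weapon trail for the FreightCalf pickaxe, a single soft grey ribbon emitter named Trail_soft.
# Anim-trail modules spawn, initialize and update short-lived particles from the system's skeletal mesh over a brief, single emitter loop.
# Width tapers by curve, while alpha fades over both a default-indexed curve and system age.
# Time of day dims alpha and tints color through a gradient, bluer at both ends and warm in between.
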NiagaraSystem /BRCosmetics/Weapons/FORT_Melee/Pickaxe_FreightCalfUniversal/FX/NS_Pickaxe_FreightCalf_Trail
dependencies {
 usedDataInterfaces [&NiagaraDataInterfaceCurve, &NiagaraDataInterfaceColorCurve, &NiagaraDataInterfaceSkeletalMesh]
 usedDynamicInputs []
 usedModules [
  &self:SystemSpawnScript.SystemScriptSource.SystemScriptGraph.NiagaraNodeAssignment_0.SetVariables_#1
  &SystemState
 ]
}
emitters {Trail_soft {
  emitterUpdateScript [
   &EmitterState(Life Cycle Mode=::NewEnumerator1, Loop Behavior=::NewEnumerator1, Loop Duration=0.24, UseLoopDelay=false)
   &FN_SpawnAnimTrail()
  ]
  particleSpawnScript [
   &FN_InitializeAnimTrailBoneIndecies() {module=FN_InitializeAnimTrailBoneIndecies001}
   &FN_InitializeAnimTrail(Color=<rgba 0.2552 0.2552 0.2552 0.7>, Lifetime=0.22, SkeletalMesh=$System.Skeletal Mesh) {module=FN_InitializeAnimTrail001}
  ]
  particleUpdateScript [
   &FN_UpdateAnimTrail(SkeletalMesh=$System.Skeletal Mesh)
   &ScaleRibbonWidth(Ribbon Width Scale=&FloatFromCurve(FloatCurve=DataInterface() {propertyValues {Curve {keys [
         {arriveTangent=-2.5249, interpMode=RCIM_Cubic, time=0.1408, value=1}
         {arriveTangent=-0.0405, interpMode=RCIM_Cubic, leaveTangent=-0.0405, tangentMode=RCTM_User, time=1, value=0.5959}
        ]}}}))
   &ScaleColor(
    Scale Alpha=&Multiply_Float(
     A=&FloatFromCurve(
      FloatCurve=DataInterface() {propertyValues {Curve {keys [{interpMode=RCIM_Cubic, tangentMode=RCTM_User, time=0, value=1}, {interpMode=RCIM_Cubic, tangentMode=RCTM_User, time=1, value=0}]}}}
      Scale Curve=&FNTimeOfDayScalar(Night Scale=0.9, Morning Scale=0.9, Evening Scale=0.7)
     )
     B=&FloatFromCurve(
      FloatCurve=DataInterface() {propertyValues {Curve {keys [{interpMode=RCIM_Cubic, tangentMode=RCTM_User, time=0.0328, value=1}, {interpMode=RCIM_Cubic, tangentMode=RCTM_User, time=0.24, value=0}]}}}
      CurveIndex=$System.Age
     )
    )
    Scale RGB=&MakeVectorFromLinearColorRGB(Color=&ColorFromCurve(
      ColorCurve=DataInterface() {propertyValues {
        AlphaCurve {keys [<at 0 1 RCIM_Cubic>]}
        BlueCurve {keys [<at 0.061 1 RCIM_Cubic>,<at 0.25 0.7656>,<at 0.5168 0.7969>,<at 0.7953 0.6562>,<at 0.9294 1 RCIM_Cubic>]}
        GreenCurve {keys [<at 0.061 0.4721 RCIM_Cubic>,<at 0.25 0.8421>,<at 0.5168 0.8631>,<at 0.7953 0.7684>,<at 0.9294 0.5059 RCIM_Cubic>]}
        RedCurve {keys [<at 0.061 0.349 RCIM_Cubic>,<at 0.25 1>,<at 0.5168 1>,<at 0.7953 1>,<at 0.9294 0.3906 RCIM_Cubic>]}
       }}
      CurveIndex=Linked() {linkedVariable {name=NPC.FortniteNPC.TimeOfDay_0-1, type=NiagaraFloat}}
     ))
    ScaleA=true
    ScaleRGB=true
   ) {module=ScaleColor002}
  ]
  rendererClasses [&NiagaraRibbonRendererProperties]
  renderers [{rendererClass=&NiagaraRibbonRendererProperties}]
 }}
summary {
 systemName=NS_Pickaxe_FreightCalf_Trail
 emitters [{emitterName=Trail_soft, rendererClasses [&NiagaraRibbonRendererProperties]}]
 userVariables [
  User.BackgroundColor {type=LinearColor, description="", defaultValue {struct=LinearColor, value=<rgba 1 1 1 1>}}
  User.RibbonIDOffset {type=NiagaraInt32}
  User.SwingSectionIndex {type=NiagaraInt32}
 ]
}
