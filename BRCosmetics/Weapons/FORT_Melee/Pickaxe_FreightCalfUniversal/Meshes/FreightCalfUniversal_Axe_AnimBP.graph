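# Animation Blueprint for the FreightCalfUniversal pickaxe, handling its front-end pose and secondary motion.
# The front end plays the equip animation and otherwise uses the input pose, feeding DefaultSlot, then b_AR_Root offset and scale adjustments.
# AnimDynamics nodes then sway the flap and two claws, weighted by the DynamicsAlpha curve.
# The melee claw rotation is unconnected, and the Large Character flag set on customization load is never read here.
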AnimBlueprint /BRCosmetics/Weapons/FORT_Melee/Pickaxe_FreightCalfUniversal/Meshes/FreightCalfUniversal_Axe_AnimBP
target_skeleton=&/FreightCalfUniversal_Axe_Skeleton
graphs [
 AnimGraph {nodes [
   "Transform (Modify) Bone - Bone: deform_claw_2"!(ComponentPose, bAlphaBoolEnabled:bool=Get bIsPlayingMeleeAnim(), Translation="0,0,0", Rotation="0,-30,0", Scale="1,1,1") -> (Pose) {
    alphaInputType=Bool
    rotationMode=BMM_Additive
    rotationSpace=BCS_BoneSpace
    translationMode=BMM_Replace
    translationSpace=BCS_BoneSpace
    alphaBoolBlend {blendInTime=0.3, blendOption=ExpInOut, blendOutTime=3}
    boneToModify {boneName=deform_claw_2}
    rotation {yaw=-30}
   }
   n1 = Sequence Player 'FreightCalfUniversal_Equip_Anim_Mainhand_WEP'() {
    bLoopAnimation=false
    sequence=/BRCosmetics/Weapons/FORT_Melee/Pickaxe_FreightCalfUniversal/Anim/FreightCalfUniversal_Equip_Anim_Mainhand_WEP
    startPosition=0.67
   }
   AnimGraphNode_SubInput_0 = Input Pose()
   n2 = Blend Poses by bool(BlendPose_0=@n1, BlendPose_1=@AnimGraphNode_SubInput_0, bActiveValue=Get bIsInFrontEnd()) {blendType=Linear}
   n3 = "Transform (Modify) Bone - Bone: b_AR_Root"!(
    ComponentPose=Local To Component(LocalPose=Slot 'DefaultSlot'(Source=@n2))
    bAlphaBoolEnabled:bool=True
    Translation="0,-0.3,0"
    Rotation="0,0,0"
    Scale="1,1,1"
   ) -> (Pose) {
    alphaInputType=Bool
    scaleMode=BMM_Additive
    scaleSpace=BCS_BoneSpace
    translationMode=BMM_Additive
    translationSpace=BCS_BoneSpace
    alphaBoolBlend {blendOption=Linear}
    boneToModify {boneName=b_AR_Root}
    translation {y=-0.3}
   }
   n4 = "Transform (Modify) Bone - Bone: b_AR_Root"(ComponentPose=@n3, bAlphaBoolEnabled=True, Translation="0,0,7", Rotation="0,0,0", Scale="1.1,1.1,1.1") {
    alphaInputType=Bool
    scaleMode=BMM_Additive
    scaleSpace=BCS_BoneSpace
    translationMode=BMM_Additive
    alphaBoolBlend {blendOption=Linear}
    boneToModify {boneName=b_AR_Root}
    scale=<1.1 1.1 1.1>
    translation {z=7}
   }
   n5 = Anim Dynamics - Bone: dyn_lower_flap!(ComponentPose=@n4, AlphaCurveName:Name=DynamicsAlpha, ExternalForce="0,0,0") -> (Pose) {
    alphaInputType=Curve
    angularBiasOverride=0.999
    angularDampingOverride=0.999
    angularSpringConstant=200000
    bOverrideAngularDamping=true
    simulationSpace=BoneRelative
    alphaBoolBlend {blendOption=Linear}
    boundBone {boneName=dyn_lower_flap}
    componentAppliedLinearAccClamp=<2000 2000 2000>
    componentLinearAccScale=<0.1 0.1 0.1>
    componentLinearVelScale=<0.1 0.1 0.1>
    physicsBodyDefinitions [{
      boundBone {boneName=dyn_lower_flap}
      boxExtents=<15 15 15>
      constraintSetup {angularLimitsMax=<0 0 5>, angularLimitsMin=<0 0 -10>, angularTarget=<0 0 0>}
      localJointOffset=<10 0 0>
     }]
    relativeSpaceBone {boneName=Melee_TwoHanded_Handle}
   }
   n6 = Anim Dynamics - Bone: dyn_claw_1!(ComponentPose=@n5, AlphaCurveName:Name=DynamicsAlpha, ExternalForce="0,0,0") -> (Pose) {
    alphaInputType=Curve
    angularBiasOverride=0.999
    angularDampingOverride=1
    angularSpringConstant=200000
    bOverrideAngularBias=true
    bOverrideAngularDamping=true
    alphaBoolBlend {blendOption=Linear}
    boundBone {boneName=dyn_claw_1}
    componentAppliedLinearAccClamp=<2000 2000 2000>
    componentLinearAccScale=<0.1 0.1 0.1>
    componentLinearVelScale=<0.1 0.1 0.1>
    physicsBodyDefinitions [{
      boundBone {boneName=dyn_claw_1}
      boxExtents=<30 30 30>
      constraintSetup {angularLimitsMax=<0 0 2>, angularLimitsMin=<0 0 -10>, angularTarget=<0 0 0>}
      localJointOffset=<-10 0 0>
     }]
    relativeSpaceBone {boneName=Melee_TwoHanded_Handle}
   }
   n7 = Anim Dynamics - Bone: dyn_claw_2!(ComponentPose=@n6, AlphaCurveName:Name=DynamicsAlpha, ExternalForce="0,0,0") -> (Pose) {
    alphaInputType=Curve
    angularBiasOverride=0.999
    angularDampingOverride=1
    angularSpringConstant=200000
    bOverrideAngularBias=true
    bOverrideAngularDamping=true
    simulationSpace=BoneRelative
    alphaBoolBlend {blendOption=Linear}
    boundBone {boneName=dyn_claw_2}
    componentAppliedLinearAccClamp=<2000 2000 2000>
    componentLinearAccScale=<0.1 0.1 0.1>
    componentLinearVelScale=<0.1 0.1 0.1>
    physicsBodyDefinitions [{
      boundBone {boneName=dyn_claw_2}
      boxExtents=<30 30 30>
      constraintSetup {angularLimitsMax=<0 0 2>, angularLimitsMin=<0 0 -20>, angularTarget=<0 0 0>}
      localJointOffset=<-5 -5 0>
     }]
    relativeSpaceBone {boneName=Melee_TwoHanded_Handle}
   }
   Output Pose(Result=Component To Local(ComponentPose=@n7)) {layerGroup=None}
  ]}
 EventGraph {nodes [
   n1 = Event OnCharacterAndPickaxeCustomizationLoaded!() -> (
    OutputDelegate
    then
    CustomBaseMesh:Skeletal Mesh Component Object Reference
    CustomParts:Array of Custom Character Part Object References
    CustomMeshParts:Array of Skeletal Mesh Component Budgeted Object References
   ) {is=title K2Node_Event}
   n2 = Get Array Item with EFort Custom Part Type(Array:Array of Custom Character Part Object References=@n1.CustomParts) -> (Output:Custom Character Part Object Reference)
   n3 = Get BodyTypesPermitted!(self:Custom Character Part Object Reference=@n2) -> (BodyTypesPermitted:EFortCustomBodyType Enum=NONE)
   n4 = Get GenderPermitted!(self:Custom Character Part Object Reference=@n2) -> (GenderPermitted:EFortCustomGender Enum=Invalid)
   > Set Large Character(Large Character:bool=AND Boolean(
     A="Equal (Enum)"(A:EFortCustomBodyType Enum=@n3, B:EFortCustomBodyType Enum=Large)
     B="Equal (Enum)"(A:EFortCustomGender Enum=@n4, B:EFortCustomGender Enum=Male)
    )) -> (Output_Get:bool)
  ]}
]
variables [Large Character {type=boolean}]
